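; Shared PCG input subgraph that supplies actors from either an ActorReferences array or the editor selection, switched by UseSelectedActors.
; The array path filters references with a not-equal soft path test and compares the remaining data count to output HasActorRefs.
; The selection path runs Get Editor Selection through an execution dependency, outputting points and an attribute set trimmed to ActorReference.
; An empty selection routes to a "No Actor Selected" error print.
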(graph Input_ActorRefsOrSelection
  (params (UseSelectedActors :type Bool) (ActorReferences :type (Array (SoftObject /Script/Engine.Actor))))
  (outputs (ActorReferences :type Param) (HasActorRefs :type Param) (SelectedActors :type Point) (SelectedActorsReferences :type Param) (UseSelectedActors :type Param))
  (bind Branch_0 (Branch :bOutputToB-param UseSelectedActors :In ActorReferences))
  (bind AttributeFilter_1 (Filter%20Attribute%20Elements :AttributeTypes (struct :Type SoftObjectPath) :bUseConstantThreshold true :Operator NotEqual :TargetAttribute @Last :ThresholdAttribute @Last))
  (bind RemoveEmptyData_1 (Remove%20Empty%20Data))
  (bind DataNum_0 (Data%20Count))
  (bind Branch_3 (Branch :bOutputToB-param UseSelectedActors :In UseSelectedActors))
  (bind GetEditorSelection_0 (Get%20Editor%20Selection))
  (bind DataNum_1 (Data%20Count))
  (bind Branch_2 (Branch))
  (bind AttributeCompareOp_0 (Attribute%20Compare%20Op :Operation Greater))
  (bind PointToAttributeSet_1 (Point%20To%20Attribute%20Set))
  (bind DeleteAttributes_0 (Delete%20Attributes :Operation KeepSelectedAttributes :SelectedAttributes "ActorReference"))
  ;[1 0 0 0.5] Error: No Actor Selected
  (bind Print%20String_1 (Print%20String :bPrefixWithNode false :bPrintPerComponent false :PrintString "No Actor Selected" :Verbosity Error))
  (connect DataNum_0 Out AttributeCompareOp_0 InA)
  (connect Branch_0 Output%20A AttributeFilter_1 In)
  (connect DataNum_1 Out Branch_2 bOutputToB)
  (connect GetEditorSelection_0 Out Branch_2 In)
  (connect RemoveEmptyData_1 Out DataNum_0 In)
  (connect GetEditorSelection_0 Out DataNum_1 In)
  (connect PointToAttributeSet_1 Out DeleteAttributes_0 In)
  (connect Branch_3 Output%20B GetEditorSelection_0 Execution%20Dependency)
  (connect Branch_2 Output%20B PointToAttributeSet_1 In)
  (connect Branch_2 Output%20A Print%20String_1 In)
  (connect AttributeFilter_1 InsideFilter RemoveEmptyData_1 In)
  (output ActorReferences AttributeFilter_1.InsideFilter)
  (output HasActorRefs AttributeCompareOp_0)
  (output SelectedActors Branch_2.Output%20B)
  (output SelectedActorsReferences DeleteAttributes_0)
  (output UseSelectedActors UseSelectedActors))
