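; PCG input graph that supplies the active editor viewport camera as point data, along with its forward vector, as a shared primitive.
; A Branch on the UseEditorViewportCamera parameter feeds the execution dependency of Get Editor Cameras, tying the camera query to that toggle.
; An index filter keeps the first camera for the point output, and its rotation's forward axis is extracted as ActiveCameraForwardVector.
; The toggle itself also passes through as an output.
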(graph Input_EditorViewportCamera
  (params (UseEditorViewportCamera :type Bool))
  (outputs (EditorViewportCameras :type Point) (UseEditorViewportCamera :type Param) (ActiveCameraForwardVector :type Param))
  (bind Branch_2 (Branch :bOutputToB-param UseEditorViewportCamera :In UseEditorViewportCamera))
  (bind GetEditorCameras_1 (Get%20Editor%20Cameras))
  (bind FilterElementsByIndex_1 (Filter%20Elements%20By%20Index :bSelectIndicesByInput false :SelectedIndices "0"))
  (bind ExtractAttribute_1 (Extract%20Attribute%20at%20Index :bForceOutputAttributeToBeInElementsDomain true :InputSource $Rotation.forward :OutputAttributeName ActiveCameraForwardVector))
  (connect FilterElementsByIndex_1 In%20Filter ExtractAttribute_1 In)
  (connect GetEditorCameras_1 Out FilterElementsByIndex_1 In)
  (connect Branch_2 Output%20B GetEditorCameras_1 Execution%20Dependency)
  (output ActiveCameraForwardVector ExtractAttribute_1)
  (output EditorViewportCameras FilterElementsByIndex_1.In%20Filter)
  (output UseEditorViewportCamera UseEditorViewportCamera))
